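# Physics asset for the Wasteland Warrior backpack cosmetic, defining collision bodies and joint limits for its dangling cloth and bullet belt.
# Box bodies cover the pack and cloth segments, which form a five-link chain from a small master capsule with modest swing and tight twist limits.
# Two bullet belt capsules chain from a master box, swinging mostly on one axis with the other swing and twist nearly locked.
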PhysicsAsset /BRCosmetics/Accessories/FORT_Backpacks/Backpack_F_MED_Wasteland_Warrior/Meshes/F_MED_Wasteland_Warrior_Pack_Physics
bodies {
 Dyn_BulletBelt2 [{shapeType=Capsule, length=2, radius=1, center=<0.4526 0 0>}]
 Dyn_BulletBelt3 [{shapeType=Capsule, length=2, radius=1, center=<0.2 0 0>}]
 Dyn_Cloth1 [{shapeType=Box, extentX=4, extentY=1, extentZ=12, center=<1.8 0 0>}]
 Dyn_Cloth2 [{shapeType=Box, extentX=4.5, extentY=1, extentZ=12, center=<2.4252 0 0>}]
 Dyn_Cloth3 [{shapeType=Box, extentX=5, extentY=1, extentZ=12, center=<2.3905 0 0>}]
 Dyn_Cloth4 [{shapeType=Box, extentX=6, extentY=1, extentZ=12, center=<2.8933 0 0>}]
 Dyn_Cloth5 [{shapeType=Box, extentX=4.5, extentY=1, extentZ=12, center=<2.171 0 0>}]
 Dyn_Master_BulletBelt [{shapeType=Box, extentX=15, extentY=1, extentZ=10, center=<8.9602 2.0321 0>}]
 Dyn_Master_Cloth [{shapeType=Capsule, length=0.505, radius=0.505}]
 pack [{shapeType=Box, extentX=30, extentY=8, extentZ=30, center=<0.0001 -5.9478 2.6104>}]
}
constraints [
 {bone1Name=Dyn_Cloth1, bone2Name=Dyn_Master_Cloth, swing1LimitDegrees=15, swing2LimitDegrees=15, twistLimitDegrees=5}
 {bone1Name=Dyn_Cloth2, bone2Name=Dyn_Cloth1, swing1LimitDegrees=15, swing2LimitDegrees=15, twistLimitDegrees=5}
 {bone1Name=Dyn_Cloth3, bone2Name=Dyn_Cloth2, swing1LimitDegrees=15, swing2LimitDegrees=15, twistLimitDegrees=5}
 {bone1Name=Dyn_Cloth4, bone2Name=Dyn_Cloth3, swing1LimitDegrees=10, swing2LimitDegrees=10, twistLimitDegrees=5}
 {bone1Name=Dyn_Cloth5, bone2Name=Dyn_Cloth4, swing1LimitDegrees=15, swing2LimitDegrees=15, twistLimitDegrees=5}
 {bone1Name=Dyn_BulletBelt2, bone2Name=Dyn_Master_BulletBelt, swing1LimitDegrees=30, swing2LimitDegrees=1, twistLimitDegrees=1}
 {bone1Name=Dyn_BulletBelt3, bone2Name=Dyn_BulletBelt2, swing1LimitDegrees=15, swing2LimitDegrees=1, twistLimitDegrees=1}
]
