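# Post-input Anim Blueprint for the female Wasteland Warrior backpack, adding physics sway to its cloth and bullet belt.
# Additive bone offsets feed a base RigidBody simulation under strong downward gravity.
# Skydive-floating and parachute-open flags blend in pack tilts and lighter-gravity RigidBody passes, with floating also rotating cloth and belt bones.
# The event graph sets Final Gravity from skydive state and fall velocity, but no simulation reads it.
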AnimBlueprint /BRCosmetics/Accessories/FORT_Backpacks/Backpack_F_MED_Wasteland_Warrior/Meshes/F_MED_Wasteland_Warrior_Pack_AnimBP
target_skeleton=&/F_MED_Wasteland_Warrior_Pack_Skeleton
graphs [
 AnimGraph {nodes [
   Local To Component(LocalPose=Use cached pose 'PoseCacheBackpack'())
   n1 = "Transform (Modify) Bone - Bone: root"(ComponentPose=Local To Component(LocalPose=Input Pose()), Translation="0,0,-1", Rotation="0,0,0") {
    rotationSpace=BCS_BoneSpace
    translationMode=BMM_Additive
    translationSpace=BCS_BoneSpace
    alphaBoolBlend {blendOption=Linear}
    boneToModify {boneName=root}
    translation {z=-1}
   }
   n2 = "Transform (Modify) Bone - Bone: Dyn_Cloth1"(ComponentPose=@n1, Alpha=1, Translation="0,2,0", Rotation="0,0,0", Scale="1,1,1") {
    rotationMode=BMM_Additive
    rotationSpace=BCS_BoneSpace
    translationMode=BMM_Additive
    translationSpace=BCS_BoneSpace
    alphaBoolBlend {blendOption=Linear}
    boneToModify {boneName=Dyn_Cloth1}
    translation {y=2}
   }
   n3 = "Transform (Modify) Bone - Bone: Dyn_Cloth2"(ComponentPose=@n2, Alpha=1, Translation="0,2,0", Rotation="0,0,0", Scale="1,1,1") {
    rotationMode=BMM_Additive
    rotationSpace=BCS_BoneSpace
    translationMode=BMM_Additive
    translationSpace=BCS_BoneSpace
    alphaBoolBlend {blendOption=Linear}
    boneToModify {boneName=Dyn_Cloth2}
    translation {y=2}
   }
   n4 = "Transform (Modify) Bone - Bone: Dyn_Cloth3"(ComponentPose=@n3, Alpha=1, Translation="0,2,0", Rotation="0,0,0", Scale="1,1,1") {
    rotationMode=BMM_Additive
    rotationSpace=BCS_BoneSpace
    translationMode=BMM_Additive
    translationSpace=BCS_BoneSpace
    alphaBoolBlend {blendOption=Linear}
    boneToModify {boneName=Dyn_Cloth3}
    translation {y=2}
   }
   n5 = "Transform (Modify) Bone - Bone: Dyn_BulletBelt2"(ComponentPose=@n4, Alpha=1, Translation="0,0,0", Rotation="0,0.2,0", Scale="1,1,1") {
    rotationMode=BMM_Additive
    rotationSpace=BCS_BoneSpace
    translationMode=BMM_Additive
    translationSpace=BCS_BoneSpace
    alphaBoolBlend {blendOption=Linear}
    boneToModify {boneName=Dyn_BulletBelt2}
    rotation {yaw=0.2}
   }
   n6 = Get bIsSkydiveFloating()
   n7 = "Transform (Modify) Bone - Bone: pack"!(ComponentPose=@n5, bAlphaBoolEnabled:bool=@n6, Translation="0,0,0", Rotation="0,0,10", Scale) -> (Pose) {
    alphaInputType=Bool
    rotationMode=BMM_Additive
    rotationSpace=BCS_BoneSpace
    translationMode=BMM_Additive
    translationSpace=BCS_BoneSpace
    alphaBoolBlend {blendOption=Linear}
    boneToModify {boneName=pack}
    rotation {roll=10}
   }
   n8 = RigidBody!(ComponentPose=@n7, Alpha, OverrideWorldGravity:Vector="0,0,-800", ExternalForce) -> (Pose) {
    bClampLinearTranslationLimitToRefPose=true
    bOverrideWorldGravity=true
    overridePhysicsAsset=&/F_MED_Wasteland_Warrior_Pack_Physics
    simulationSpace=BaseBoneSpace
    alphaBoolBlend {blendOption=Linear}
    baseBoneRef {boneName=root}
    componentAppliedLinearAccClamp=<1000 1000 1000>
    componentLinearAccScale=<0.5 0.4 0.3>
    componentLinearVelScale=<0.5 0.5 0.3>
    overrideWorldGravity {z=-800}
   }
   n9 = RigidBody!(ComponentPose=@n8, bAlphaBoolEnabled:bool=@n6, ExternalForce) -> (Pose) {
    alphaInputType=Bool
    bClampLinearTranslationLimitToRefPose=true
    bOverrideWorldGravity=true
    overridePhysicsAsset=&F_MED_Wasteland_Warrior_Pack_Physics
    simulationSpace=BaseBoneSpace
    alphaBoolBlend {blendOption=Linear}
    baseBoneRef {boneName=root}
    componentAppliedLinearAccClamp=<1000 1000 1000>
    componentLinearAccScale=<0.5 0.4 0.5>
    componentLinearVelScale=<0.5 0.5 0.5>
    overrideWorldGravity {z=-200}
   }
   n10 = "Transform (Modify) Bone - Bone: Dyn_Cloth1"!(ComponentPose=@n9, bAlphaBoolEnabled:bool=@n6, Translation="0,0,0", Rotation="0,-30,0", Scale="1,1,1") -> (Pose) {
    alphaInputType=Bool
    rotationMode=BMM_Additive
    rotationSpace=BCS_BoneSpace
    translationMode=BMM_Additive
    translationSpace=BCS_BoneSpace
    alphaBoolBlend {blendOption=Linear}
    boneToModify {boneName=Dyn_Cloth1}
    rotation {yaw=-30}
   }
   n11 = "Transform (Modify) Bone - Bone: Dyn_Master_BulletBelt"!(ComponentPose=@n10, bAlphaBoolEnabled:bool=@n6, Translation="0,0,0", Rotation="0,-15,0", Scale="1,1,1") -> (Pose) {
    alphaInputType=Bool
    rotationMode=BMM_Additive
    rotationSpace=BCS_BoneSpace
    translationMode=BMM_Additive
    translationSpace=BCS_BoneSpace
    alphaBoolBlend {blendOption=Linear}
    boneToModify {boneName=Dyn_Master_BulletBelt}
    rotation {yaw=-15}
   }
   n12 = Get bIsParachuteOpened()
   n13 = RigidBody(ComponentPose=@n11, bAlphaBoolEnabled=@n12) {
    alphaInputType=Bool
    bClampLinearTranslationLimitToRefPose=true
    bOverrideWorldGravity=true
    overridePhysicsAsset=&F_MED_Wasteland_Warrior_Pack_Physics
    simulationSpace=BaseBoneSpace
    alphaBoolBlend {blendOption=Linear}
    baseBoneRef {boneName=root}
    componentAppliedLinearAccClamp=<1000 1000 1000>
    componentLinearAccScale=<0.5 0.4 1>
    componentLinearVelScale=<0.5 0.5 1>
    overrideWorldGravity {z=-200}
   }
   n14 = "Transform (Modify) Bone - Bone: pack"(ComponentPose=@n13, bAlphaBoolEnabled=@n12, Translation="0,0,0", Rotation="0,0,10") {
    alphaInputType=Bool
    rotationMode=BMM_Additive
    rotationSpace=BCS_BoneSpace
    translationMode=BMM_Additive
    translationSpace=BCS_BoneSpace
    alphaBoolBlend {blendOption=Linear}
    boneToModify {boneName=pack}
    rotation {roll=10}
   }
   Output Pose(Result=Component To Local(ComponentPose=@n14)) {layerGroup=None}
   n15 = RigidBody!(
    ComponentPose=Local To Component(LocalPose=Use cached pose 'PoseCacheBackpack'())
    Alpha:float
    OverrideWorldGravity:Vector="0,0,-800"
    ExternalForce
   ) -> (Pose) {
    bClampLinearTranslationLimitToRefPose=true
    bOverrideWorldGravity=true
    overridePhysicsAsset=&F_MED_Wasteland_Warrior_Pack_Physics
    simulationSpace=BaseBoneSpace
    alphaBoolBlend {blendOption=Linear}
    baseBoneRef {boneName=root}
    componentAppliedLinearAccClamp=<1000 1000 1000>
    componentLinearAccScale=<0.5 0.4 0.3>
    componentLinearVelScale=<0.5 0.5 0.3>
    overrideWorldGravity {z=-800}
   }
   Save cached pose 'PoseCacheBackpack'(Pose=Component To Local(ComponentPose=@n15)) {cacheName=PoseCacheBackpack}
   Save cached pose 'PoseCacheBackpack'(Pose=Component To Local()) {cacheName=PoseCacheBackpack}
   Get Gravity overide() -> (Gravity overide:Vector)
  ]}
 EventGraph {nodes [
   Event BlueprintUpdateAnimation()
   > n1 = Branch(Condition=Get bIsSkyDiving())
   > n2 = Branch(Condition=Get bIsParachuteOpened())
   n3 = Branch(execute=@n2.else, Condition="float > float"(A=Get PawnVelocityZ(), B=10) {is=title K2Node_CallFunction})
   > Set Final Gravity(Final Gravity:Vector="0, 0,200") -> (Output_Get:Vector)
   Set Final Gravity(execute=@n3.else, Final Gravity="0, 0,-1000")
   Set Final Gravity(execute=@n1.else, Final Gravity="0, 0,-980")
   TryGetPawnOwner()
   Get GravityOverride!(self:Custom Character Part Anim Instance Frogman M Object Reference) -> (GravityOverride)
  ]}
]
variables [
 Active Value {type=boolean}
 Component Linear Acc Scale {type=Vector}
 Component Linear Vel Scale {type=Vector}
 Final Gravity {type=Vector}
 Gravity overide {type=Vector}
 Is in the Air {type=boolean}
]
